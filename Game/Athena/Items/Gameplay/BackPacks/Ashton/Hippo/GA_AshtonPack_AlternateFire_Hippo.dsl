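; Gameplay ability for the alternate fire of the Ashton Hippo backpack item, launching an alt projectile with an explosive bow fire montage.
; On activation it starts the montage and immediately spawns the projectile at the pawn's aim location, along base aim rotation, with zero gravity and scalable-float speed.
; It then fires a cue, sends the AltFire event, applies a cooldown effect and ends without waiting on montage callbacks.
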(event Ability|EventActivateAbility
  (bind _asfort_player_pawn (Utilities|Casting|CastToFortPlayerPawn (Ability|GetAvatarActorfromActorInfo)))
  (Ability|Tasks|PlayMontageAndWait "None" "/Game/Animation/Game/MainPlayer/Combat/Guns/ExplosiveBow/ExplosiveBow_Fire_M.ExplosiveBow_Fire_M"
    (:then
      (bind _returnvalue (Utility|GetAimLocationNearestPawn _asfort_player_pawn))
      (Projectile|SpawnProjectile :ProjectileClass "/Game/Athena/Items/Gameplay/BackPacks/Ashton/Hippo/B_Prj_Ashton_Hippo_Alt.B_Prj_Ashton_Hippo_Alt_C" :RequestedBy _asfort_player_pawn :SpawnLocation _returnvalue :SpawnRotation (Pawn|GetBaseAimRotation _asfort_player_pawn) :SpawnDirection (Pawn|GetBaseAimRotation _asfort_player_pawn) :bUseDefaultPhysics false :InitialSpeed (ScalableFloat|GetValueatLevel (Variables|Default|GetRow_Speed)) :GravityScale 0.0 :ChargePercent 0.0)
      (Ability|ExecuteGameplayCueOnOwner (Variables|Default|GetAltFireCue))
      (Ability|SendGameplayEvent "(TagName=\"Event.Athena.ExplosiveBow.AltFire\")")
      (Ability|ApplyGameplayEffectToOwner (Variables|Default|GetGE_FireCooldown)))
    (:OnCompleted
      _)
    (:OnBlendOut
      _)
    (:OnInterrupted
      _)
    (:OnCancelled
      _))
  (Ability|EndAbility))
