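; GA_Ashton_Hippo_Primary is the primary-fire gameplay ability for the Ashton Hippo backpack item.
; On commit, it applies the alt-fire cooldown gameplay effect to the owner, briefly blocking alt fire whenever primary fire is used, then calls the parent commit logic.
; Activate and end events are disabled stubs, so the rest of the firing behaviour is left to the parent ability class.
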(event Ability|EventActivateAbility
  (|CallParentFunction) ; This node is disabled and will not be called.\nDrag off pins to build functionality.
)

(event Ability|EventOnEndAbility
  (params (bWasCancelled :type Boolean))
  (|CallParentFunction bWasCancelled) ; This node is disabled and will not be called.\nDrag off pins to build functionality.
)

(event Ability|EventCommitExecute
  ;[1 1 1] Apply a GE to block the alt fire when ever we fire for a short duration
  (Ability|ApplyGameplayEffectToOwner (Variables|Default|GetGE_AltfireCooldownPrimaryFire))
  (|CallParentFunction))
